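# Niagara system for the popcorn in the Short Scare player emote, built from one sprite emitter of settled corn pops.
# A single burst, gated to the emitter's first loop by an HLSL expression, scatters small sprites over a thin, flat disc.
# Each sprite gets a random size, one of two warm tints via a rounded random lerp, and dynamic material parameters.
# Particles ignore lifetime expiry, allowing infinitely lived ones to die when the emitter deactivates.
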NiagaraSystem /BRCosmetics/Animation/Game/MainPlayer/Emotes/Short_Scare/VFX/NS_Popcorn02_Timed_Shorts_Scare_Emote
dependencies {usedDataInterfaces [&NiagaraDataInterfaceCurve], usedDynamicInputs [], usedModules [&SystemState]}
emitters {Settled_CornPops {
  emitterUpdateScript [
   &EmitterState()
   &SpawnBurst_Instantaneous(Spawn Count=50, Spawn Probability=HlslExpression() {hlslExpression="Emitter.LoopCount == 0"}, Use Spawn Probability=true)
  ]
  particleSpawnScript [
   /Niagara/Modules/Spawn/Initialization/InitializeParticle(
    Color=&LerpLinearColors(
     StartColor=<rgba 0.95 0.7323 0.4275 1>
     EndColor=<rgba 0.95 0.8313 0.665 1>
     LerpFactor=/Niagara/DynamicInputs/Integer/V2/Round_Float(Float:NiagaraFloat=&RandomRangeFloat(Minimum=0, Maximum=1))
    )
    Lifetime=5
    Module.Write Color=true
    Module.Write Lifetime=false
    Module.Write Mass=true
    Module.Write Material Random=true
    Module.Write Position=false
    Module.Write Scale=false
    Module.Write SpriteRotation=false
    Module.Write SpriteSize=true
    Sprite Size=&Vector2DFromFloat(Value=&RandomRangeFloat(Minimum=5, Maximum=7))
   )
   &DynamicMaterialParameters(Module.Param0WriteEnabled=true, Module.Param1WriteEnabled=true, Module.Param2WriteEnabled=true, Module.Param3WriteEnabled=true)
   &CylinderLocation(
    Cylinder Height=0
    Cylinder Radius=&RandomRangeFloat(Minimum=1, Maximum=18)
    MaskSpawnGroup=false
    Module.Enable Lathe Profile=false
    Module.Override Randomness=false
    Module.Surface Only=false
    Offset=&RandomRangeVector(Minimum=<0 0 -1>, Maximum=<0 0 0.1>)
   )
   &AddVelocity()
  ]
  particleUpdateScript [
   &ParticleState(Kill Particles When Lifetime Has Elapsed=false, Let Infinitely Lived Particles Die When Emitter Deactivates=true)
   &SolveForcesAndVelocity(Clamp Velocity=false, Limit Acceleration=false)
  ]
  rendererClasses [&NiagaraSpriteRendererProperties]
  renderers [{rendererClass=&NiagaraSpriteRendererProperties}]
 }}
summary {
 systemName=NS_Popcorn02_Timed_Shorts_Scare_Emote
 emitters [{emitterName=Settled_CornPops, rendererClasses [&NiagaraSpriteRendererProperties]}]
 userVariables []
}
