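# Niagara system that throws popcorn-colored sprites for the enter phase of the player's Short Scare emote.
# Two burst emitters, a larger immediate one and a smaller, tighter delayed one, spawn particles on flat cylinder areas and launch them away from a point below.
# Particles fall under gravity, bounce with collision aging, then shrink and fade.
# A random dynamic material parameter is set to zero when collision reports resting, although rest state is disabled.
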NiagaraSystem /BRCosmetics/Animation/Game/MainPlayer/Emotes/Short_Scare/VFX/NS_Popcorn_Enter_Emote_Short_Scare
dependencies {usedDataInterfaces [&NiagaraDataInterfaceCurve], usedDynamicInputs [], usedModules [&SystemState]}
emitters {
 Burst01 {
  emitterUpdateScript [
   &EmitterState()
   &SpawnBurst_Instantaneous(Spawn Count=25, Use Spawn Probability=false)
  ]
  particleSpawnScript [
   /Niagara/Modules/Spawn/Initialization/InitializeParticle(
    Color=&LerpLinearColors(
     StartColor=<rgba 0.95 0.7719 0.5225 1>
     EndColor=<rgba 0.95 0.851 0.7125 1>
     LerpFactor=/Niagara/DynamicInputs/Integer/V2/Round_Float(Float:NiagaraFloat=&RandomRangeFloat(Minimum=0, Maximum=1))
    )
    Lifetime=&RandomRangeFloat(Minimum=1, Maximum=2)
    Module.Write Color=true
    Module.Write Lifetime=true
    Module.Write Mass=true
    Module.Write Material Random=true
    Module.Write Position=false
    Module.Write Scale=false
    Module.Write SpriteRotation=false
    Module.Write SpriteSize=true
    Sprite Size=&Vector2DFromFloat(Value=&RandomRangeFloat(Minimum=4, Maximum=6))
   )
   &DynamicMaterialParameters(
    Index 0 Param 1=&RandomRangeFloat(Minimum=0.75, Maximum=1.5)
    Module.Param0WriteEnabled=true
    Module.Param1WriteEnabled=true
    Module.Param2WriteEnabled=true
    Module.Param3WriteEnabled=true
   )
   &CylinderLocation(
    Cylinder Height=0
    Cylinder Radius=&Lerp_Float(A=1, B=60, Alpha=$Particles.MaterialRandom)
    MaskSpawnGroup=false
    Override Seed=false
    UseOwnerScale=true
   ) {module=CylinderLocation001}
   &AddVelocityFromPoint(
    Module.Use Falloff Distance=false
    Origin Offset=<0 0 -80.9454>
    Velocity Strength=&RandomRangeFloat(Minimum=100, Maximum=300)
   )
  ]
  particleUpdateScript [
   &ParticleState()
   &ScaleSpriteSize(Scale Factor=&Vector2DFromFloat(Value=&FloatFromCurve(FloatCurve=DataInterface() {propertyValues {Curve {keys [<at 0.85 1 RCIM_Cubic>,<at 1 0>]}}})))
   &GravityForce()
   &Collision(
    Advanced Aging Rate=0.5
    Age Colliding Particles=true
    Enable Rest State=false
    Friction=0.35
    Module.EnableMaxCollisionCount=false
    Module.Randomize Collision Normal Vector=false
    Restitution=0.3
   )
   &DynamicMaterialParameters(
    Index 0 Param 1=&MakeCustomFloatFromBool(Bool=$Transient.CollisionIsResting, True Float=0, False Float=HlslExpression() {hlslExpression=Particles.DynamicMaterialParameter.x})
    Module.Param0WriteEnabled=true
    Module.Param1WriteEnabled=false
    Module.Param2WriteEnabled=false
    Module.Param3WriteEnabled=false
   ) {module=DynamicMaterialParameters001}
   &SolveForcesAndVelocity(Clamp Velocity=false, Limit Acceleration=false)
   &ScaleColor(
    Scale Alpha=&FloatFromCurve(FloatCurve=DataInterface() {propertyValues {Curve {keys [<at 0.7 1 RCIM_Cubic>,<at 1 0>]}}})
    ScaleA=true
    ScaleRGB=true
   )
  ]
  rendererClasses [&NiagaraSpriteRendererProperties]
  renderers [{rendererClass=&NiagaraSpriteRendererProperties}]
 }
 Burst02 {
  emitterUpdateScript [
   &EmitterState()
   &SpawnBurst_Instantaneous(Spawn Count=10, Spawn Time=0.385, Use Spawn Probability=false)
  ]
  particleSpawnScript [
   &InitializeParticle(
    Color=&LerpLinearColors(
     StartColor=<rgba 0.95 0.7719 0.5225 1>
     EndColor=<rgba 0.95 0.851 0.7125 1>
     LerpFactor=&Round_Float(Float:NiagaraFloat=&RandomRangeFloat(Minimum=0, Maximum=1))
    )
    Lifetime=&RandomRangeFloat(Minimum=1, Maximum=2)
    Module.Write Color=true
    Module.Write Lifetime=true
    Module.Write Mass=true
    Module.Write Material Random=true
    Module.Write Position=false
    Module.Write Scale=false
    Module.Write SpriteRotation=false
    Module.Write SpriteSize=true
    Sprite Size=&Vector2DFromFloat(Value=&RandomRangeFloat(Minimum=4, Maximum=6))
   )
   &DynamicMaterialParameters(
    Index 0 Param 1=&RandomRangeFloat(Minimum=0.75, Maximum=1.5)
    Module.Param0WriteEnabled=true
    Module.Param1WriteEnabled=true
    Module.Param2WriteEnabled=true
    Module.Param3WriteEnabled=true
   )
   &CylinderLocation(
    Cylinder Height=0
    Cylinder Radius=&Lerp_Float(A=1, B=25, Alpha=$Particles.MaterialRandom)
    MaskSpawnGroup=false
    Offset Mode:ENiagara_OffsetMode=::NewEnumerator0
    Override Seed=false
    UseOwnerScale=true
   ) {module=CylinderLocation001}
   &AddVelocityFromPoint(
    Module.Use Falloff Distance=false
    Origin Offset=<0 0 -20>
    Velocity Strength=&RandomRangeFloat(Minimum=100, Maximum=200)
   )
  ]
  particleUpdateScript [
   &ParticleState()
   &ScaleSpriteSize(Scale Factor=&Vector2DFromFloat(Value=&FloatFromCurve(FloatCurve=DataInterface() {propertyValues {Curve {keys [<at 0.85 1 RCIM_Cubic>,<at 1 0>]}}})))
   &GravityForce()
   &Collision(
    Advanced Aging Rate=0.5
    Age Colliding Particles=true
    Enable Rest State=false
    Friction=0.35
    Module.EnableMaxCollisionCount=false
    Module.Randomize Collision Normal Vector=false
    Restitution=0.3
   )
   &DynamicMaterialParameters(
    Index 0 Param 1=&MakeCustomFloatFromBool(Bool=$Transient.CollisionIsResting, True Float=0, False Float=HlslExpression() {hlslExpression=Particles.DynamicMaterialParameter.x})
    Module.Param0WriteEnabled=true
    Module.Param1WriteEnabled=false
    Module.Param2WriteEnabled=false
    Module.Param3WriteEnabled=false
   ) {module=DynamicMaterialParameters001}
   &SolveForcesAndVelocity(Clamp Velocity=false, Limit Acceleration=false)
   &ScaleColor(
    Scale Alpha=&FloatFromCurve(FloatCurve=DataInterface() {propertyValues {Curve {keys [<at 0.7 1 RCIM_Cubic>,<at 1 0>]}}})
    ScaleA=true
    ScaleRGB=true
   )
  ]
  rendererClasses [&NiagaraSpriteRendererProperties]
  renderers [{rendererClass=&NiagaraSpriteRendererProperties}]
 }
}
summary {
 systemName=NS_Popcorn_Enter_Emote_Short_Scare
 emitters [
  {emitterName=Burst01, rendererClasses [&NiagaraSpriteRendererProperties]}
  {emitterName=Burst02, rendererClasses [&NiagaraSpriteRendererProperties]}
 ]
 userVariables []
}
